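# Translucent unlit material for a klaxon warning light's rotating cone, red by default.
# Color times brightness is divided by eye adaptation to counter auto exposure.
# Opacity combines a gradient falloff with inverted Fresnel, adding depth fade except at low and medium quality.
# RotateAboutAxis spins vertices around local Z at the object origin over time, with an actor-position phase offset to desynchronise lights.
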Material /WildEstateTech/TechArt/KlaxonLight/M_WildEstate_KlaxonLightCone
graph {
 properties {bAutomaticallySetUsageInEditor=false, blendMode=BLEND_Translucent, shadingModel=MSM_Unlit, usedShadingModels=Unlit}
 nodes [
  Divide(
   A=Multiply(A=VectorParameter() {parameterName=Color, defaultValue=<rgba 1 0 0 1>}.RGB, B=ScalarParameter() {defaultValue=1, parameterName=Brightness})
   B=EyeAdaptation().EyeAdaptation
  )
  n1 = Power(
   Base=MaterialFunctionCall() {materialFunction=&LinearGradient}.VGradient
   Exp=ScalarParameter() {defaultValue=1, parameterName=FalloffCurve}
  )
  n2 = LinearInterpolate(
   B=MaterialFunctionCall(
    Brightness=ScalarParameter() {defaultValue=1, parameterName=Fresnel.Brightness}
    Exponent=ScalarParameter() {defaultValue=1, parameterName=Fresnel.Exp}
   ) {materialFunction=&MF_Fresnel_FX}.Fresnel Inverted
   Alpha=ScalarParameter() {defaultValue=1, parameterName=Fresnel.Blend}
  ) {constA=1}
  n3 = Saturate(None=Multiply(A=Multiply(A=@n1, B=@n2), B=ScalarParameter() {defaultValue=0.5, parameterName=OpacityMult}))
  QualitySwitch(Default=DepthFade(Opacity=@n3, FadeDistance=ScalarParameter() {defaultValue=20, parameterName=DepthFade}), Low=@n3, Medium=@n3)
  n4 = Constant3Vector() {constant=<rgba 0 0 1 0>}
  n5 = ActorPositionWS()
  n6 = Add(A=Time(), B=Frac(None=Length(None=@n5) {is=class MaterialExpressionLength}))
  n7 = ScalarParameter() {defaultValue=0.1, parameterName=RotationsPerSecond}
  RotateAboutAxis(
   NormalizedRotationAxis=Transform(None=Normalize(VectorInput=@n4)) {transformSourceType=TRANSFORMSOURCE_Local}
   RotationAngle=Frac(None=Multiply(A=@n6, B=@n7))
   PivotPoint=TransformPosition(None=Constant3Vector()) {transformType=TRANSFORMPOSSOURCE_World}
   Position=WorldPosition().XYZ
  )
 ]
}
